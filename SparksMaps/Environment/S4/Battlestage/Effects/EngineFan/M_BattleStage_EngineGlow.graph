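# Unlit, two-sided material for the Battlestage engine fan glow in the S4 environment.
# A time-driven sine oscillates the radius of an exponential radial gradient, which is tinted by a Color parameter (blue by default) and scaled by Strength for a pulsing glow.
# A separate branch builds a generated band in radial coordinates, masked by a linear gradient with parameterized falloff.
# The graph shows no link between the two branches.
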Material /SparksMaps/Environment/S4/Battlestage/Effects/EngineFan/M_BattleStage_EngineGlow
graph {
 properties {bAutomaticallySetUsageInEditor=false, shadingModel=MSM_Unlit, twoSided=true, usedShadingModels=Unlit}
 nodes [
  Constant()
  n1 = Multiply(A=Time()) {constB=4}
  n2 = MaterialFunctionCall(Radius=LinearInterpolate(Alpha=Sine(None=@n1)) {constA=0.46, constB=0.51}, Density=Constant() {r=1}) {materialFunction=&RadialGradientExponential}
  Multiply(
   A=Multiply(A=VectorParameter() {parameterName=Color, defaultValue=<rgba 0 0.236 0.7604 1>}.RGB, B=Multiply(A=@n2.RadialGradientExponential))
   B=ScalarParameter() {defaultValue=10, parameterName=Strength}
  )
  n3 = MaterialFunctionCall(Swizzle Coordinate Output=StaticBool() {value=true}) {materialFunction=&VectorToRadialValue}
  n4 = MaterialFunctionCall(UV Channel=@n3.Radial Coordinates) {materialFunction=&LinearGradient}
  n5 = Divide(
   A=Subtract(A=OneMinus(None=@n4.UGradient), B=ScalarParameter() {defaultValue=0.1, parameterName=GradSub})
   B=ScalarParameter() {defaultValue=0.1, parameterName=GradDivid}
  )
  Multiply(
   A=MaterialFunctionCall!(Width=Constant() {r=0.1}, Sharpness, Offset=Constant() {r=-0.2}, Direction Switch, Compare, Input Coordinates=@n3.Radial Coordinates) {materialFunction=/Engine/Functions/Engine_MaterialFunctions03/Procedurals/GeneratedBand}.Result
   B=Multiply(A=@n4.UGradient, B=Saturate(None=@n5))
  )
 ]
}
